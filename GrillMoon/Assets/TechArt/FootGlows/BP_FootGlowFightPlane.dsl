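; Foot-glow plane actor for the GrillMoon fight, feeding the foot positions of two fighters (MaskMaker and Demon) to a world-space glow material.
; Construction sets that material and moves the plane to the virtual floor height.
; Setter functions assign the skeletal actors.
; UpdateMPC writes each valid fighter's foot socket positions, pushed along a mirrored rotated socket axis, into MPC_GrillMoon, republishing floor height only when Demon is valid.
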(fn SetMistMylarSkeletalActor
  (params (MistMylar :type Skeletal%20Mesh%20Actor%20Object%20Reference))
  (Variables|Default|SetSkeletonMaskMaker MistMylar))

(fn SetSekeletalTargets
  (params (SkeletonA :type Skeletal%20Mesh%20Actor%20Object%20Reference) (SkeletonB :type Skeletal%20Mesh%20Actor%20Object%20Reference))
  (Variables|Default|SetSkeletonMaskMaker SkeletonA)
  (Variables|Default|SetSkeletonDemon SkeletonB))

(fn SetUnluckyRollSkeletalActor
  (params (UnluckyRoll :type Skeletal%20Mesh%20Actor%20Object%20Reference))
  (Variables|Default|SetSkeletonDemon UnluckyRoll))

(fn UpdateMPC
  (bind _skeletondemon (Variables|Default|GetSkeletonDemon))
  (bind _skeletalmeshcomponent (Variables|/Script/Engine.SkeletalMeshActor|GetSkeletalMeshComponent _skeletondemon))
  (bind _skeletonmaskmaker (Variables|Default|GetSkeletonMaskMaker))
  (bind _skeletalmeshcomponent_2 (Variables|/Script/Engine.SkeletalMeshActor|GetSkeletalMeshComponent _skeletonmaskmaker))
  (Utilities|IsValid _skeletonmaskmaker
    (:Is%20Valid
      (Rendering|Material|SetVectorParameterValue "/GrillMoon_Countdown/Assets/FX/MPC_GrillMoon.MPC_GrillMoon" "MaskMaker_FootPos_L" (Math|Conversions|ToLinearColor(Vector) (+ (Transformation|GetSocketLocation _skeletalmeshcomponent_2 "foot_l") (* (Math|Vector|RotateVectorAroundAxis (Math|Vector|GetUpVector (Transformation|GetSocketRotation _skeletalmeshcomponent_2 "foot_l")) -9e+01 (Math|Vector|GetForwardVector (Transformation|GetSocketRotation _skeletalmeshcomponent_2 "foot_l"))) (Variables|Default|GetFootFrontOffset)))) :!class /Script/Engine.KismetMaterialLibrary)
      (Rendering|Material|SetVectorParameterValue "/GrillMoon_Countdown/Assets/FX/MPC_GrillMoon.MPC_GrillMoon" "MaskMaker_FootPos_R" (Math|Conversions|ToLinearColor(Vector) (+ (Transformation|GetSocketLocation _skeletalmeshcomponent_2 "foot_r") (* (Math|Vector|RotateVectorAroundAxis (Math|Vector|GetUpVector (Transformation|GetSocketRotation _skeletalmeshcomponent_2 "foot_r")) 9e+01 (Math|Vector|GetForwardVector (Transformation|GetSocketRotation _skeletalmeshcomponent_2 "foot_r"))) (Variables|Default|GetFootFrontOffset)))) :!class /Script/Engine.KismetMaterialLibrary))
    (:Is%20Not%20Valid
      _))
  (Utilities|IsValid _skeletondemon
    (:Is%20Valid
      (Rendering|Material|SetVectorParameterValue "/GrillMoon_Countdown/Assets/FX/MPC_GrillMoon.MPC_GrillMoon" "Demon_FootPos_L" (Math|Conversions|ToLinearColor(Vector) (+ (Transformation|GetSocketLocation _skeletalmeshcomponent "foot_l") (* (Math|Vector|RotateVectorAroundAxis (Math|Vector|GetUpVector (Transformation|GetSocketRotation _skeletalmeshcomponent "foot_l")) -9e+01 (Math|Vector|GetForwardVector (Transformation|GetSocketRotation _skeletalmeshcomponent "foot_l"))) (Variables|Default|GetFootFrontOffset)))) :!class /Script/Engine.KismetMaterialLibrary)
      (Rendering|Material|SetVectorParameterValue "/GrillMoon_Countdown/Assets/FX/MPC_GrillMoon.MPC_GrillMoon" "Demon_FootPos_R" (Math|Conversions|ToLinearColor(Vector) (+ (Transformation|GetSocketLocation _skeletalmeshcomponent "foot_r") (* (Math|Vector|RotateVectorAroundAxis (Math|Vector|GetUpVector (Transformation|GetSocketRotation _skeletalmeshcomponent "foot_r")) 9e+01 (Math|Vector|GetForwardVector (Transformation|GetSocketRotation _skeletalmeshcomponent "foot_r"))) (Variables|Default|GetFootFrontOffset)))) :!class /Script/Engine.KismetMaterialLibrary)
      (Rendering|Material|SetScalarParameterValue "/GrillMoon_Countdown/Assets/FX/MPC_GrillMoon.MPC_GrillMoon" "Fight_VirtualFloorHeight" (Variables|Default|GetVirtualFloorHeight) :!class /Script/Engine.KismetMaterialLibrary))))

(fn ConstructionScript
  (bind (_returnvalue_x _returnvalue_y) (Math|Vector|BreakVector (Transformation|GetActorLocation)))
  (bind _virtualfloorheight (Variables|Default|GetVirtualFloorHeight))
  (Rendering|Material|SetMaterial (Variables|Default|GetPlane) 0 "/GrillMoon/Assets/TechArt/FootGlows/MI_FootGlow_GrillMoon_WorldSpace.MI_FootGlow_GrillMoon_WorldSpace")
  (Transformation|SetWorldLocation (Variables|Default|GetPlane) (Math|Vector|MakeVector _returnvalue_x _returnvalue_y _virtualfloorheight) false false)
  (Rendering|Material|SetScalarParameterValue "/GrillMoon_Countdown/Assets/FX/MPC_GrillMoon.MPC_GrillMoon" "Fight_VirtualFloorHeight" _virtualfloorheight :!class /Script/Engine.KismetMaterialLibrary))
