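; Foot glow Blueprint that places two glow planes beneath the left and right feet of a target SkeletalMeshActor.
; The construction script scales both planes and gives each its own dynamic material instance.
; Position functions project the foot_l and foot_r socket XY onto a virtual floor height, randomizing yaw on every call.
; Alpha functions push per-foot glow values into Master.Alpha, and setters assign the followed skeleton.
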(fn SetSekeletonFronActorAndSetTarget
  (params (Actor :type Actor%20Object%20Reference))
  (bind _asskeletal_mesh_actor (Utilities|Casting|CastToSkeletalMeshActor Actor))
  (Variables|Default|SetSkeletonToFollow _asskeletal_mesh_actor))

(fn SetSkeletonTarget
  (params (SkeletonToFollow :type Skeletal%20Mesh%20Actor%20Object%20Reference))
  (Variables|Default|SetSkeletonToFollow SkeletonToFollow))

(fn UpdateAlphaLeft
  (Rendering|Material|SetScalarParameterValue (Variables|Default|GetMID_L) "Master.Alpha" (Variables|Default|GetGlowAlpha_Left) :!class /Script/Engine.MaterialInstanceDynamic))

(fn UpdateAlphaRight
  (Rendering|Material|SetScalarParameterValue (Variables|Default|GetMID_R) "Master.Alpha" (Variables|Default|GetGlowAlpha_Right) :!class /Script/Engine.MaterialInstanceDynamic))

(fn UpdateAlphas
  (CallFunction|UpdateAlphaRight)
  (CallFunction|UpdateAlphaLeft))

(fn UpdateGlowPosLeft
  (bind (_returnvalue_x _returnvalue_y) (Math|Vector|BreakVector (Transformation|GetSocketLocation (Variables|/Script/Engine.SkeletalMeshActor|GetSkeletalMeshComponent (Variables|Default|GetSkeletonToFollow)) "foot_l")))
  (bind _plane_l (Variables|Default|GetPlane_L))
  (Transformation|SetWorldLocation _plane_l (Math|Vector|MakeVector _returnvalue_x _returnvalue_y (Variables|Default|GetVirtualFloorHeight)) false false)
  (Transformation|SetWorldRotation _plane_l (Math|Rotator|MakeRotator 0.0 0.0 (* (Math|Random|RandomFloat) 3.6e+02)) false false))

(fn UpdateGlowPosRight
  (bind (_returnvalue_x _returnvalue_y) (Math|Vector|BreakVector (Transformation|GetSocketLocation (Variables|/Script/Engine.SkeletalMeshActor|GetSkeletalMeshComponent (Variables|Default|GetSkeletonToFollow)) "foot_r")))
  (bind _plane_r (Variables|Default|GetPlane_R))
  (Transformation|SetWorldLocation _plane_r (Math|Vector|MakeVector _returnvalue_x _returnvalue_y (Variables|Default|GetVirtualFloorHeight)) false false)
  (Transformation|SetWorldRotation _plane_r (Math|Rotator|MakeRotator 0.0 0.0 (* (Math|Random|RandomFloat) 3.6e+02)) false false))

(fn ConstructionScript
  (Utilities|FlowControl|Sequence
    (:then_0
      (bind _returnvalue (Rendering|Material|CreateDynamicMaterialInstance (Variables|Default|GetMaterial) "None" "None" :!class /Script/Engine.KismetMaterialLibrary))
      (bind _output_get (Variables|Default|SetMID_R _returnvalue))
      (Rendering|Material|SetMaterial (Variables|Default|GetPlane_R) 0 _output_get)
      (Transformation|SetRelativeScale3D (Variables|Default|GetPlane_R) (Math|Conversions|ToVector(Float) (Variables|Default|GetScale))))
    (:then_1
      (bind _returnvalue_2 (Rendering|Material|CreateDynamicMaterialInstance (Variables|Default|GetMaterial) "None" "None" :!class /Script/Engine.KismetMaterialLibrary))
      (bind _output_get_2 (Variables|Default|SetMID_L _returnvalue_2))
      (Rendering|Material|SetMaterial (Variables|Default|GetPlane_L) 0 _output_get_2)
      (Transformation|SetRelativeScale3D (Variables|Default|GetPlane_L) (Math|Conversions|ToVector(Float) (Variables|Default|GetScale))))))
