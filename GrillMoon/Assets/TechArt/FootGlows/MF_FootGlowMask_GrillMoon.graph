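# Material function that builds glow and pulse masks around a character's foot on a virtual floor.
# MaskOut is a sphere mask at the rounded foot position (rounded X feeds both X and Y), scaled by HeightFadeMask, a floor-distance fade optionally maxed below the plane.
# PulseMaskA and PulseMaskB are floor-height sphere masks distorted by world-aligned Perlin noise, static and time-panned, plus input warp blended by radius.
# A uses half the panned warp strength of B.
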MaterialFunction /GrillMoon/Assets/TechArt/FootGlows/MF_FootGlowMask_GrillMoon
graph {nodes [
  n1 = WorldPosition()
  n2 = TextureObject() {texture=/Game/Global/Textures/Default/Noise/3xPerlin_M}
  n3 = ScalarParameter() {defaultValue=19000, desc=Use to create variation across entire floor plane as characters move their feet, parameterName=SecondaryWarpTiling}
  n4 = MaterialFunctionCall(TextureObject=@n2, TextureSize=@n3) {materialFunction=&WorldAlignedTexture}
  n5 = ScalarParameter() {defaultValue=256, desc=Use to create variation across entire floor plane as characters move their feet, parameterName=SecondaryWarpIntensity}
  n6 = Add(A=@n1.XYZ, B=Multiply(A=Subtract(A=@n4.Z Texture) {constB=0.5}, B=@n5))
  n7 = ScalarParameter() {defaultValue=256, parameterName=Pulse.Secondary.Tiling}
  n8 = Time()
  n9 = Add(
   A=WorldPosition().XYZ
   B=Multiply(
    A=FunctionInput(Preview=@n8) {bUsePreviewValueAsDefault=true, inputName=Time, inputType=FunctionInput_Scalar, sortPriority=11}
    B=ScalarParameter() {defaultValue=1, parameterName=Pulse.Secondary.PanSpeed}
   )
  )
  n10 = Multiply(
   A=Subtract(A=MaterialFunctionCall(TextureObject=@n2, TextureSize=@n7, WorldPosition=@n9) {materialFunction=&WorldAlignedTexture}.Z Texture) {constB=0.5}
   B=ScalarParameter() {defaultValue=50, parameterName=Pulse.Secondary.Intensity}
  )
  n11 = Add(A=@n6, B=Multiply(A=@n10) {constB=0.5})
  n12 = FunctionInput() {bUsePreviewValueAsDefault=true, inputName=PulseUVsWarp, sortPriority=9}
  n13 = FunctionInput() {inputName=FootPosition, sortPriority=1}
  n14 = ScalarParameter() {defaultValue=10, parameterName=Round}
  n15 = Multiply(A=@n13, B=@n14) {constB=10}
  n16 = MaterialFunctionCall(Float3=Divide(A=Round(None=@n15), B=@n14)) {materialFunction=&BreakOutFloat3Components}
  n17 = MaterialFunctionCall(
   X=@n16.R
   Y=@n16.R
   Z=MaterialFunctionCall(Float3=@n13) {materialFunction=&BreakOutFloat3Components}.B
  ) {materialFunction=&MakeFloat3}
  n18 = FunctionInput() {bUsePreviewValueAsDefault=true, inputName=PulseRadius, inputType=FunctionInput_Scalar, sortPriority=10, previewValue=<100 0 0 1>}
  n19 = ScalarParameter() {
   defaultValue=2
   desc=Use to create variation across entire floor plane as characters move their feet
   parameterName=Pulse.WarpBlendRadiusDivisor
  }
  n20 = LinearInterpolate(
   A=Add(A=@n12, B=FunctionInput() {bUsePreviewValueAsDefault=true, inputName=PulseUVsWarpBlend, sortPriority=9})
   B=@n12
   Alpha=SphereMask(A=@n1.XYZ, B=@n17.Result, Radius=Divide(A=@n18, B=@n19)) {attenuationRadius=100, hardnessPercent=0.5}
  )
  n21 = FunctionInput() {inputName=VirtualFloorHeight, inputType=FunctionInput_Scalar, sortPriority=2}
  n22 = AppendVector(A=ComponentMask(None=@n17.Result) {g=true, r=true}, B=@n21)
  FunctionOutput(None=SphereMask(A=Add(A=@n11, B=@n20), B=@n22, Radius=@n18) {attenuationRadius=100, hardnessPercent=0.5}) {outputName=PulseMaskA, sortPriority=5}
  n23 = Add(A=@n6, B=@n10)
  FunctionOutput(None=SphereMask(A=Add(A=@n20, B=@n23), B=@n22, Radius=@n18) {attenuationRadius=100, hardnessPercent=0.5}) {outputName=PulseMaskB, sortPriority=6}
  n24 = ComponentMask(None=@n17.Result) {b=true}
  n25 = Divide(A=Distance(A=@n21, B=@n24), B=ScalarParameter() {defaultValue=1500, parameterName=FootMask.HeightFadeDistance})
  n26 = Power(Base=OneMinus(None=Saturate(None=@n25)), Exp=ScalarParameter() {defaultValue=1.5, parameterName=FootMask.HeightFadePow}) {constExponent=1.5}
  n27 = StaticSwitchParameter(True=If(A=@n21, B=@n24, "A > B"=Constant() {r=1}, "A == B"=@n26, "A < B"=@n26), False=@n26) {defaultValue=true, parameterName=UseMaxValueWhenFootBelowPlane}
  n28 = MaterialFunctionCall(Z=FunctionInput() {bUsePreviewValueAsDefault=true, inputName=FootMaskZOffset, inputType=FunctionInput_Scalar, sortPriority=3}) {materialFunction=&MakeFloat3}
  n29 = FunctionInput() {bUsePreviewValueAsDefault=true, inputName=SphereMaskRadius, inputType=FunctionInput_Scalar, sortPriority=5, previewValue=<100 0 0 1>}
  n30 = FunctionInput() {bUsePreviewValueAsDefault=true, inputName=SphereMaskHardness, inputType=FunctionInput_Scalar, sortPriority=6, previewValue=<0.5 0 0 1>}
  n31 = Multiply(A=@n27, B=SphereMask(A=WorldPosition().XYZ, B=Add(A=@n17.Result, B=@n28.Result), Radius=@n29, Hardness=@n30))
  n32 = NamedRerouteUsage()
  n33 = If(A=@n32, "A > B"=@n32, "A == B"=Constant() {r=1}, "A < B"=Add(A=@n32)) {equalsThreshold=0.001}
  FunctionOutput(None=Saturate(None=Multiply(A=@n31, B=@n33))) {outputName=MaskOut}
  FunctionOutput(None=@n27) {outputName=HeightFadeMask}
  NamedRerouteDeclaration(None=FunctionInput() {bUsePreviewValueAsDefault=true, inputName=ValueMultiOverride, inputType=FunctionInput_Scalar, sortPriority=20}) {name=ValOverrideINPUT, nodeColor=<rgba 0 1 0.6953 1>}
  SphereMask(B=@n22, Radius=@n18) {attenuationRadius=100, hardnessPercent=0.5}
  SphereMask(B=@n22, Radius=@n18) {attenuationRadius=100, hardnessPercent=0.5}
 ]}
