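# Additive unlit world-space material that draws tinted glows, pulses and cracks around the feet and bodies of MaskMaker and Demon in the GrillMoon fight.
# MPC_GrillMoon positions and virtual floor height feed MF_FootGlowMask calls, with per-limb overrides.
# Those masks drive warped sine pulses, alongside wider XY reveal spheres, world-aligned cracks and panning noise, shared through named reroutes.
# Emissive divides by eye adaptation; depth-faded opacity passes through a custom clip node.
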Material /GrillMoon/Assets/TechArt/FootGlows/M_FootGlow_GrillMoon_WorldSpace
graph {
 properties {bAutomaticallySetUsageInEditor=false, blendMode=BLEND_Additive, shadingModel=MSM_Unlit, twoSided=true, usedShadingModels=Unlit}
 nodes [
  MaterialExpressionMultiply_28 = Multiply(
   A=DepthFade(Opacity=Constant() {r=1}, FadeDistance=ScalarParameter() {defaultValue=250, parameterName=Opacity.DepthFadeDistance}) {fadeDistanceDefault=250}
   B=ScalarParameter() {defaultValue=1, parameterName=Opacity.Master}
  )
  MaterialExpressionConstant_0 = Constant() {r=0.005}
  Custom!(A=@MaterialExpressionMultiply_28, Threshold=@MaterialExpressionConstant_0) {
   code="clip(A-Threshold); return A;"
   showCode=true
   inputs [
    {inputName=A, input {expression=&self:MaterialExpressionMultiply_28}}
    {inputName=Threshold, input {expression=&self:MaterialExpressionConstant_0}}
   ]
  }
  n1 = VectorParameter() {parameterName=Tint.MaskMaker, defaultValue=<rgba 0 0.6852 1 1>}
  n2 = NamedRerouteUsage()
  n3 = VectorParameter() {parameterName=Tint.Demon, defaultValue=<rgba 1 0.2227 0 1>}
  n4 = NamedRerouteUsage()
  n5 = Multiply(A=NamedRerouteUsage(), B=Add(A=Multiply(A=@n1.RGB, B=@n2), B=Multiply(A=@n3.RGB, B=@n4)))
  n6 = Multiply(A=@n1.RGB, B=NamedRerouteUsage())
  n7 = Multiply(A=@n3.RGB, B=NamedRerouteUsage())
  n8 = NamedRerouteUsage()
  n9 = Multiply(
   A=Power(Base=@n8, Exp=ScalarParameter() {defaultValue=1, parameterName=Cracks.Micro.RevealMaskPower})
   B=ScalarParameter() {defaultValue=1, parameterName=Cracks.Micro.RevealMaskMulti}
  )
  n10 = Saturate(None=Multiply(A=Saturate(None=@n9), B=NamedRerouteUsage()))
  n11 = NamedRerouteUsage()
  n12 = ScalarParameter() {defaultValue=1, parameterName=MysticNoise.SubtractiveMask.Multi}
  n13 = LinearInterpolate(A=@n10, B=Subtract(A=@n10, B=Multiply(A=@n11, B=@n12)), Alpha=NamedRerouteUsage())
  n14 = Multiply(
   A=LinearInterpolate(B=Add(A=@n6, B=@n7), Alpha=ScalarParameter() {defaultValue=1, parameterName=Cracks.ColorBlend}) {constA=1}
   B=LinearInterpolate(B=ScalarParameter() {defaultValue=1, parameterName=Cracks.Emissive.Boost}, Alpha=Saturate(None=@n13))
  ) {desc=Cracks Color}
  Divide(A=Multiply(A=Add(A=@n5, B=@n14), B=ScalarParameter() {defaultValue=1, parameterName=Emissive.Boost}), B=EyeAdaptation().EyeAdaptation)
  n15 = CollectionParameter() {collection=/GrillMoon_Countdown/Assets/FX/MPC_GrillMoon, parameterName=Fight_VirtualFloorHeight}
  n16 = ScalarParameter() {parameterName=MaskOffsetZ}
  n17 = ScalarParameter() {defaultValue=500, parameterName=FootGlowRadius}
  n18 = ScalarParameter() {defaultValue=1, parameterName=FootGlowHardness}
  n19 = NamedRerouteUsage()
  n20 = ScalarParameter() {defaultValue=500, parameterName=FootPulseRadius}
  n21 = ScalarParameter() {desc="0 = no override, -1 remove glow, 1 force glow on", group=Overrides, parameterName=Demon.FootOverride.Left, sliderMax=1, sliderMin=-1}
  n22 = MaterialFunctionCall(
   TextureObject=TextureObject() {texture=/GrillMoon/Assets/TechArt/StaticClouds/T_Cloud_Noise_GrillMoon}
   TextureSize=ScalarParameter() {defaultValue=1000, parameterName=PulseWarpTiling}
  ) {materialFunction=&WorldAlignedTexture}
  n23 = Multiply(A=Multiply(A=Subtract(A=@n22.XYZ Texture) {constB=0.5}) {constB=2}, B=ScalarParameter() {parameterName=PulseWarpIntensity})
  n24 = MaterialFunctionCall!(
   FootPosition=ComponentMask(None=CollectionParameter() {collection=&MPC_GrillMoon, parameterName=Demon_FootPos_L}) {b=true, g=true, r=true}
   VirtualFloorHeight=@n15
   FootMaskZOffset=@n16
   SphereMaskRadius=@n17
   SphereMaskHardness=@n18
   PulseUVsWarp=@n23
   PulseUVsWarpBlend=@n19
   PulseRadius=@n20
   Time
   ValueMultiOverride=@n21
  ) {materialFunction=&/MF_FootGlowMask_GrillMoon}
  n25 = ScalarParameter() {desc="0 = no override, -1 remove glow, 1 force glow on", group=Overrides, parameterName=Demon.FootOverride.Right, sliderMax=1, sliderMin=-1}
  n26 = MaterialFunctionCall(
   FootPosition=ComponentMask(None=CollectionParameter() {collection=&MPC_GrillMoon, parameterName=Demon_FootPos_R}) {b=true, g=true, r=true}
   VirtualFloorHeight=@n15
   FootMaskZOffset=@n16
   SphereMaskRadius=@n17
   SphereMaskHardness=@n18
   PulseUVsWarp=@n23
   PulseUVsWarpBlend=@n19
   PulseRadius=@n20
   ValueMultiOverride=@n25
  ) {materialFunction=&MF_FootGlowMask_GrillMoon}
  n27 = Max(A=@n24.MaskOut, B=@n26.MaskOut)
  n28 = ScalarParameter() {desc="0 = no override, -1 remove glow, 1 force glow on", group=Overrides, parameterName=Demon.BodyOverride, sliderMax=1, sliderMin=-1}
  n29 = MaterialFunctionCall(
   FootPosition=CollectionParameter() {collection=&MPC_GrillMoon, parameterName=DemonMaker_Body_Pos}
   VirtualFloorHeight=@n15
   FootMaskZOffset=@n16
   SphereMaskRadius=ScalarParameter() {defaultValue=500, group=Overrides, parameterName=Demon.Body.Radius, sliderMax=1, sliderMin=-1}
   SphereMaskHardness=@n18
   PulseUVsWarp=@n23
   PulseUVsWarpBlend=@n19
   PulseRadius=@n20
   ValueMultiOverride=@n28
  ) {materialFunction=&MF_FootGlowMask_GrillMoon}
  NamedRerouteDeclaration(None=Saturate(None=Max(A=@n27, B=@n29.MaskOut))) {name=DemonMask, nodeColor=<rgba 1 0.4453 0 1>}
  n30 = Add(A=TextureCoordinate()) {constB=-0.5}
  n31 = Time()
  n32 = TextureObject() {texture=&T_Cloud_Noise_GrillMoon}
  n33 = TextureSample(
   UVs=Add(
    A=MaterialFunctionCall(Vector or UVs=@n30) {materialFunction=&VectorToRadialValue}.Radial Coordinates
    B=Multiply(A=@n31) {constB=-0.01}
   )
   Tex=@n32
  )
  n34 = TextureSample(Tex=@n32) {samplerType=SAMPLERTYPE_LinearColor}
  n35 = Divide(A=Add(A=AppendVector(A=@n33.R, B=@n33.R), B=AppendVector(A=@n34.R, B=@n34.R)))
  NamedRerouteDeclaration(None=Multiply(A=Subtract(A=@n35) {constB=0.5}) {constB=2}) {name=CloudNoise, nodeColor=<rgba 0.0312 0 1 1>}
  n36 = NamedRerouteUsage()
  n37 = MaterialFunctionCall(UVs=Multiply(A=TextureCoordinate(), B=@n36), CenterPosition=Divide(A=@n36), Density=Constant() {r=1.5}) {materialFunction=&RadialGradientExponential}
  n38 = NamedRerouteUsage()
  n39 = Sine(None=Add(A=Multiply(A=@n37.RadialGradientExponential) {constB=2}, B=Multiply(A=@n38) {constB=0.4}))
  NamedRerouteDeclaration(None=Saturate(None=@n39)) {name=PulseMask, nodeColor=<rgba 0.6172 1 0 1>}
  n40 = ScalarParameter() {
   desc="0 = no override, -1 remove glow, 1 force glow on"
   group=Overrides
   parameterName=MaskMaker.FootOverride.Left
   sliderMax=1
   sliderMin=-1
  }
  n41 = If(A=@n40, "A > B"=@n40, "A == B"=Constant() {r=1}, "A < B"=Add(A=@n40)) {equalsThreshold=0.001}
  n42 = WorldPosition()
  n43 = ScalarParameter() {defaultValue=4000, parameterName=PlaneReveal.Radius}
  n44 = ScalarParameter() {defaultValue=0.7, parameterName=PlaneReveal.Hardness}
  n45 = SphereMask(
   A=@n42.XY
   B=ComponentMask(None=CollectionParameter() {collection=&MPC_GrillMoon, parameterName=MaskMaker_FootPos_L}) {g=true, r=true}
   Radius=@n43
   Hardness=@n44
  )
  n46 = ScalarParameter() {
   desc="0 = no override, -1 remove glow, 1 force glow on"
   group=Overrides
   parameterName=MaskMaker.FootOverride.Right
   sliderMax=1
   sliderMin=-1
  }
  n47 = If(A=@n46, "A > B"=@n46, "A == B"=Constant() {r=1}, "A < B"=Add(A=@n46)) {equalsThreshold=0.001}
  n48 = SphereMask(
   A=@n42.XY
   B=ComponentMask(None=CollectionParameter() {collection=&MPC_GrillMoon, parameterName=MaskMaker_FootPos_R}) {g=true, r=true}
   Radius=@n43
   Hardness=@n44
  )
  n49 = ComponentMask(None=CollectionParameter() {collection=&MPC_GrillMoon, parameterName=MaskMaker_Body_Pos}) {g=true, r=true}
  n50 = ScalarParameter() {defaultValue=4000, group=Overrides, parameterName=Body.RevealRadius}
  n51 = ScalarParameter() {defaultValue=0.7, group=Overrides, parameterName=Body.RevealHardness}
  n52 = ScalarParameter() {desc="0 = no override, -1 remove glow, 1 force glow on", group=Overrides, parameterName=MaskMaker.BodyOverride, sliderMax=1, sliderMin=-1}
  n53 = Max(
   A=Max(A=Multiply(A=@n41, B=@n45), B=Multiply(A=@n47, B=@n48))
   B=Multiply(A=SphereMask(A=@n42.XY, B=@n49, Radius=@n50, Hardness=@n51), B=@n52)
  )
  n54 = SphereMask(
   A=@n42.XY
   B=ComponentMask(None=CollectionParameter() {collection=&MPC_GrillMoon, parameterName=Demon_FootPos_L}) {g=true, r=true}
   Radius=@n43
   Hardness=@n44
  )
  n55 = ScalarParameter() {desc="0 = no override, -1 remove glow, 1 force glow on", group=Overrides, parameterName=Demon.FootOverride.Left, sliderMax=1, sliderMin=-1}
  n56 = If(A=@n55, "A > B"=@n55, "A == B"=Constant() {r=1}, "A < B"=Add(A=@n55)) {equalsThreshold=0.001}
  n57 = SphereMask(
   A=@n42.XY
   B=ComponentMask(None=CollectionParameter() {collection=&MPC_GrillMoon, parameterName=Demon_FootPos_R}) {g=true, r=true}
   Radius=@n43
   Hardness=@n44
  )
  n58 = ScalarParameter() {desc="0 = no override, -1 remove glow, 1 force glow on", group=Overrides, parameterName=Demon.FootOverride.Right, sliderMax=1, sliderMin=-1}
  n59 = If(A=@n58, "A > B"=@n58, "A == B"=Constant() {r=1}, "A < B"=Add(A=@n58)) {equalsThreshold=0.001}
  n60 = ComponentMask(None=CollectionParameter() {collection=&MPC_GrillMoon, parameterName=DemonMaker_Body_Pos}) {g=true, r=true}
  n61 = ScalarParameter() {defaultValue=4000, group=Overrides, parameterName=Body.RevealRadius, sliderMax=1, sliderMin=-1}
  n62 = ScalarParameter() {defaultValue=0.7, group=Overrides, parameterName=Body.RevealHardness, sliderMax=1, sliderMin=-1}
  n63 = ScalarParameter() {desc="0 = no override, -1 remove glow, 1 force glow on", group=Overrides, parameterName=Demon.BodyOverride, sliderMax=1, sliderMin=-1}
  n64 = Max(
   A=Max(A=Multiply(A=@n54, B=@n56), B=Multiply(A=@n57, B=@n59))
   B=Multiply(A=SphereMask(A=@n42.XY, B=@n60, Radius=@n61, Hardness=@n62), B=@n63)
  )
  NamedRerouteDeclaration(None=Max(A=@n53, B=@n64)) {name=RevealMask.Master, nodeColor=<rgba 0.9688 0 1 1>}
  n65 = CollectionParameter() {collection=&MPC_GrillMoon, parameterName=Fight_VirtualFloorHeight}
  n66 = ScalarParameter() {parameterName=MaskOffsetZ}
  n67 = ScalarParameter() {defaultValue=1, parameterName=FootGlowHardness}
  n68 = NamedRerouteUsage()
  n69 = ScalarParameter() {defaultValue=500, parameterName=FootPulseRadius}
  n70 = ScalarParameter() {desc="0 = no override, -1 remove glow, 1 force glow on", group=Overrides, parameterName=MaskMaker.BodyOverride, sliderMax=1, sliderMin=-1}
  n71 = MaterialFunctionCall(
   FootPosition=CollectionParameter() {collection=&MPC_GrillMoon, parameterName=MaskMaker_Body_Pos}
   VirtualFloorHeight=@n65
   FootMaskZOffset=@n66
   SphereMaskRadius=ScalarParameter() {defaultValue=500, group=Overrides, parameterName=MaskMaker.Body.Radius, sliderMax=1, sliderMin=-1}
   SphereMaskHardness=@n67
   PulseUVsWarp=@n23
   PulseUVsWarpBlend=@n68
   PulseRadius=@n69
   ValueMultiOverride=@n70
  ) {materialFunction=&MF_FootGlowMask_GrillMoon}
  n72 = ScalarParameter() {defaultValue=500, parameterName=FootGlowRadius}
  n73 = ScalarParameter() {
   desc="0 = no override, -1 remove glow, 1 force glow on"
   group=Overrides
   parameterName=MaskMaker.FootOverride.Left
   sliderMax=1
   sliderMin=-1
  }
  n74 = MaterialFunctionCall(
   FootPosition=ComponentMask(None=CollectionParameter() {collection=&MPC_GrillMoon, parameterName=MaskMaker_FootPos_L}) {b=true, g=true, r=true}
   VirtualFloorHeight=@n65
   FootMaskZOffset=@n66
   SphereMaskRadius=@n72
   SphereMaskHardness=@n67
   PulseUVsWarp=@n23
   PulseUVsWarpBlend=@n68
   PulseRadius=@n69
   ValueMultiOverride=@n73
  ) {materialFunction=&MF_FootGlowMask_GrillMoon}
  n75 = ScalarParameter() {
   desc="0 = no override, -1 remove glow, 1 force glow on"
   group=Overrides
   parameterName=MaskMaker.FootOverride.Right
   sliderMax=1
   sliderMin=-1
  }
  n76 = MaterialFunctionCall(
   FootPosition=ComponentMask(None=CollectionParameter() {collection=&MPC_GrillMoon, parameterName=MaskMaker_FootPos_R}) {b=true, g=true, r=true}
   VirtualFloorHeight=@n65
   FootMaskZOffset=@n66
   SphereMaskRadius=@n72
   SphereMaskHardness=@n67
   PulseUVsWarp=@n23
   PulseUVsWarpBlend=@n68
   PulseRadius=@n69
   ValueMultiOverride=@n75
  ) {materialFunction=&MF_FootGlowMask_GrillMoon}
  n77 = Max(A=Max(A=@n71.PulseMaskA, B=Max(A=@n74.PulseMaskA, B=@n76.PulseMaskA)), B=Max(A=Max(A=@n24.PulseMaskA, B=@n26.PulseMaskA), B=@n29.PulseMaskA))
  n78 = Saturate(None=@n77)
  n79 = Multiply(A=Power(Base=OneMinus(None=@n78)) {constExponent=1.5}, B=ScalarParameter() {defaultValue=5, parameterName=Pulse.NumberOfWaves}) {constB=5}
  n80 = Multiply(A=Multiply(A=NamedRerouteUsage()) {constB=-1}, B=ScalarParameter() {defaultValue=1, parameterName=Pulse.Speed})
  n81 = ConstantBiasScale(None=Sine(None=Add(A=@n79, B=@n80)))
  n82 = ScalarParameter() {defaultValue=1, parameterName=Pulse.Contrast}
  n83 = Power(Base=Saturate(None=Multiply(A=@n81, B=@n77)), Exp=@n82)
  n84 = ScalarParameter() {defaultValue=1, parameterName=Pulse.Multi}
  n85 = Max(A=Max(A=@n71.PulseMaskB, B=Max(A=@n74.PulseMaskB, B=@n76.PulseMaskB)), B=Max(A=Max(A=@n24.PulseMaskB, B=@n26.PulseMaskB), B=@n29.PulseMaskB))
  n86 = Saturate(None=@n85)
  n87 = Multiply(A=Power(Base=OneMinus(None=@n86)) {constExponent=1.2}, B=ScalarParameter() {defaultValue=5, parameterName=Pulse.Secondary.NumberOfWaves}) {constB=8}
  n88 = Multiply(A=Multiply(A=NamedRerouteUsage()) {constB=-1}, B=ScalarParameter() {defaultValue=1, parameterName=Pulse.Speed})
  n89 = ConstantBiasScale(None=Sine(None=Add(A=@n87, B=@n88)))
  n90 = Power(Base=Saturate(None=Multiply(A=@n89, B=@n85)), Exp=@n82)
  n91 = Saturate(None=Multiply(A=Multiply(A=@n83, B=@n84), B=Multiply(A=@n90, B=@n84)))
  n92 = Multiply(A=@n24.HeightFadeMask, B=@n24.HeightFadeMask)
  n93 = Multiply(A=@n26.HeightFadeMask, B=@n26.HeightFadeMask)
  n94 = Max(
   A=Max(A=Multiply(A=@n74.HeightFadeMask, B=@n74.HeightFadeMask), B=Multiply(A=@n76.HeightFadeMask, B=@n76.HeightFadeMask))
   B=Max(A=Max(A=@n92, B=@n93), B=Multiply(A=@n29.HeightFadeMask, B=@n29.HeightFadeMask))
  )
  NamedRerouteDeclaration(None=Multiply(A=Multiply(A=@n91, B=@n94), B=ScalarParameter() {defaultValue=0.7, parameterName=PulseOpacity})) {name=PulseMasterMask, nodeColor=<rgba 1 0.4453 0 1>}
  NamedRerouteDeclaration(None=@n53) {name=RevealMask.MistMylar, nodeColor=<rgba 0.9688 0 1 1>}
  NamedRerouteDeclaration(None=@n64) {name=RevealMask.Demon, nodeColor=<rgba 0.9688 0 1 1>}
  n95 = TextureObject() {samplerType=SAMPLERTYPE_Masks, texture=/Game/Athena/WinterArt/Textures/Cracks_05_M}
  n96 = MaterialFunctionCall(TextureObject=@n95, TextureSize=ScalarParameter() {defaultValue=512, parameterName=Cracks.Micro.Tiling}, WorldPosition=Reroute()) {materialFunction=&WorldAlignedTexture}
  n97 = Power(
   Base=OneMinus(None=ComponentMask(None=@n96.XYZ Texture) {r=true})
   Exp=ScalarParameter() {defaultValue=1, parameterName=Cracks.Micro.MaskPower}
  )
  n98 = ScalarParameter() {defaultValue=1, parameterName=Cracks.Micro.MaskMulti}
  NamedRerouteDeclaration(None=Saturate(None=Multiply(A=@n97, B=@n98))) {name=Cracks.Out.Micro, nodeColor=<rgba 0.4297 0 1 1>}
  n99 = TextureCoordinate()
  n100 = VectorParameter() {parameterName=MysticNoise.Tiling, channelNames {a=MysticGlobal, b="", g=B, r=A}, defaultValue=<rgba 0.8 1 0 1>}
  n101 = Multiply(A=@n99, B=Multiply(A=@n100.A, B=@n100.A))
  n102 = TextureCoordinate()
  n103 = Multiply(A=ScalarParameter() {defaultValue=0.5, parameterName=MysticNoise.UVWarp.Tiling}, B=@n100.MysticGlobal)
  n104 = NamedRerouteUsage()
  n105 = ScalarParameter() {defaultValue=1, parameterName=MysticNoise.UVWarp.Speed}
  n106 = TextureObjectParameter() {
   parameterName=MysticNoise.NormalUVWarp
   samplerType=SAMPLERTYPE_Normal
   texture=/Game/Effects/Fort_Effects/Textures/Fire/T_WispyFireNoise_Soft_N
  }
  n107 = TextureSample(UVs=Panner(Coordinate=Multiply(A=@n102, B=@n103), Time=Multiply(A=@n104, B=@n105) {constB=0}) {speedX=-0.05, speedY=0.1}, Tex=@n106) {samplerType=SAMPLERTYPE_Normal}
  n108 = ScalarParameter() {defaultValue=0.1, parameterName=MysticNoise.UVWarp.Intensity}
  n109 = Multiply(A=AppendVector(A=@n107.R, B=@n107.R), B=@n108) {constB=0.1}
  n110 = NamedRerouteUsage()
  n111 = VectorParameter() {parameterName=MysticNoise.Speed, channelNames {a=Y, b=X, g=Y, r=X}, defaultValue=<rgba 0.1 0.1 -0.08 -0.12>}
  n112 = TextureObjectParameter() {parameterName=MysticNoise.Texture, texture=&/T_Energy_GrillMoon}
  n113 = TextureSample(UVs=Panner(Coordinate=Add(A=@n101, B=@n109), Time=@n110, Speed=AppendVector(A=@n111.X, B=@n111.X)) {speedX=0.1, speedY=0.1}, Tex=@n112)
  n114 = Multiply(A=@n99, B=Multiply(A=@n100.B, B=@n100.B))
  n115 = Multiply(A=AppendVector(A=@n107.G, B=@n107.G), B=@n108) {constB=0.1}
  n116 = TextureSample(UVs=Panner(Coordinate=Add(A=@n114, B=@n115), Time=@n110, Speed=AppendVector(A=@n111.X, B=@n111.X)) {speedX=-0.08, speedY=-0.12}, Tex=@n112)
  n117 = Multiply(A=Add(A=@n113.R, B=@n116.R), B=ScalarParameter() {defaultValue=0.3, parameterName=MysticNoise.LayerOpacity}) {constB=0.3}
  n118 = Multiply(A=@n113.R, B=@n116.R)
  n119 = Power(Base=Saturate(None=Add(A=@n117, B=@n118)), Exp=ScalarParameter() {defaultValue=1, parameterName=MysticNoise.Contrast})
  n120 = ScalarParameter() {defaultValue=1, parameterName=MysticNoise.Boost}
  NamedRerouteDeclaration(None=Saturate(None=Multiply(A=@n119, B=@n120))) {name=MysticNoise, nodeColor=<rgba 0.6406 1 0 1>}
  n121 = MaterialFunctionCall(TextureObject=@n95, TextureSize=ScalarParameter() {defaultValue=512, parameterName=Cracks.Tiling}, WorldPosition=Reroute()) {materialFunction=&WorldAlignedTexture}
  n122 = Power(Base=OneMinus(None=ComponentMask(None=@n121.XYZ Texture) {r=true}), Exp=ScalarParameter() {defaultValue=1, parameterName=Cracks.MaskPower})
  n123 = ScalarParameter() {defaultValue=1, parameterName=Cracks.MaskMulti}
  NamedRerouteDeclaration(None=Saturate(None=Multiply(A=@n122, B=@n123))) {name=Cracks.Out, nodeColor=<rgba 0.4297 0 1 1>}
  n124 = Power(Base=Max(A=@n77, B=@n85), Exp=ScalarParameter() {defaultValue=1, parameterName=Pulse.WarpMask.Power})
  n125 = ScalarParameter() {defaultValue=1, parameterName=Pulse.WarpMask.Multi}
  NamedRerouteDeclaration(None=Saturate(None=Multiply(A=@n124, B=@n125))) {name=WarpedPulseMask, nodeColor=<rgba 1 0.4453 0 1>}
  NamedRerouteDeclaration(None=@n94) {name=FootDistanceMask, nodeColor=<rgba 1 0.3281 0 1>}
  NamedRerouteDeclaration(None=Multiply(A=Time(), B=ScalarParameter() {defaultValue=1, parameterName=Master.TimeScalar})) {name=Master.TimeOUT, nodeColor=<rgba 1 0.2344 0 1>}
  n126 = Add(A=TextureCoordinate()) {constB=-0.5}
  n127 = NamedRerouteUsage()
  n128 = Add(
   A=MaterialFunctionCall(Vector or UVs=@n126) {materialFunction=&VectorToRadialValue}.Radial Coordinates
   B=MaterialFunctionCall(Y=Multiply(A=@n127) {constB=-0.02}) {materialFunction=&MakeFloat2}.Result
  )
  n129 = Multiply(A=Multiply(A=Multiply(A=@n37.RadialGradientExponential, B=@n126), B=@n39)) {constB=0.2}
  n130 = NamedRerouteUsage()
  n131 = TextureSample(UVs=Panner(Coordinate=Add(A=@n128, B=@n129) {constB=0}, Time=@n130) {speedX=0.01}) {samplerType=SAMPLERTYPE_LinearColor, texture=/GrillMoon/Assets/TechArt/StaticClouds/T_ChromeSwirl_GrillMoon}
  n132 = Panner(Time=@n130) {speedX=-0.01, speedY=0.01}
  n133 = Multiply(A=@n131.R) {constB=0.2}
  n134 = TextureSample(UVs=MaterialFunctionCall(X=Add(A=@n132, B=@n133), Y=ComponentMask(None=@n132) {g=true}) {materialFunction=&MakeFloat2}.Result) {texture=&T_Cloud_Noise_GrillMoon}
  n135 = Add(A=Divide(A=Subtract(A=@n131.R) {constB=0.5}) {constB=6}, B=ComponentMask(None=Subtract(A=@n134.RGB) {constB=0.5}) {g=true, r=true})
  n136 = MaterialFunctionCall(X=ScalarParameter() {defaultValue=1, parameterName=Master.Scale.X}, Y=ScalarParameter() {defaultValue=1, parameterName=Master.Scale.Y}) {materialFunction=&MakeFloat2}
  n137 = Add(A=Multiply(A=@n135) {constB=0.1}, B=Multiply(A=TextureCoordinate(), B=@n136.Result))
  n138 = Multiply(
   A=ScalarParameter() {defaultValue=0.4, parameterName=Wisps.Radius}
   B=ScalarParameter() {defaultValue=1, parameterName=Master.Alpha, sliderMax=1}
  )
  n139 = ScalarParameter() {defaultValue=1.5, parameterName=Wisps.Density}
  n140 = Divide(A=@n136.Result)
  NamedRerouteDeclaration(None=Multiply(
    A=MaterialFunctionCall(UVs=@n137, CenterPosition=@n140, Radius=@n138, Density=@n139) {materialFunction=&RadialGradientExponential}.RadialGradientExponential
    B=ScalarParameter() {defaultValue=1, parameterName=Wisps.Opacity}
   )) {name=WispyMask, desc=Test, nodeColor=<rgba 1 0 0.4453 1>}
  NamedRerouteDeclaration(None=MaterialFunctionCall(X=ScalarParameter() {defaultValue=1, parameterName=Master.Scale.X}, Y=ScalarParameter() {defaultValue=1, parameterName=Master.Scale.Y}) {materialFunction=&MakeFloat2}.Result) {name=Master.Scale.V2, nodeColor=<rgba 1 0 0.375 1>}
  NamedRerouteDeclaration(None=@n33.RGB) {name=RadialCloudNoise, nodeColor=<rgba 0.0312 0 1 1>}
  NamedRerouteDeclaration(None=Multiply(A=ScalarParameter() {parameterName=PulseWarpIntensity2}, B=@n107.RGB)) {name=UVWarpPulseEdge, nodeColor=<rgba 0 1 0.1562 1>}
  n141 = Max(A=Multiply(A=@n71.HeightFadeMask, B=@n71.HeightFadeMask), B=@n71.MaskOut)
  n142 = Max(A=@n74.MaskOut, B=@n76.MaskOut)
  NamedRerouteDeclaration(None=Saturate(None=Max(A=@n141, B=@n142))) {name=MaskMakerMask, nodeColor=<rgba 0.125 0 1 1>}
  NamedRerouteDeclaration(None=@n135) {name=WispyMaskNoise, desc=Test, nodeColor=<rgba 1 0 0.4453 1>}
  NamedRerouteUsage()
  NamedRerouteUsage()
  NamedRerouteUsage()
  Power(Base=Max(A=NamedRerouteUsage(), B=NamedRerouteUsage()), Exp=ScalarParameter() {defaultValue=1, parameterName=Cracks.SubtractiveMask.Pwr})
  n143 = Multiply(A=Saturate(None=@n8), B=NamedRerouteUsage())
  Saturate(None=Add(A=Saturate(None=@n143), B=@n10))
  Saturate()
  ScalarParameter() {defaultValue=0.7, parameterName=PlaneReveal.Hardness}
  ScalarParameter() {defaultValue=4000, parameterName=PlaneReveal.Radius}
  ScalarParameter() {defaultValue=0.7, parameterName=PlaneReveal.Hardness}
  ScalarParameter() {defaultValue=4000, parameterName=PlaneReveal.Radius}
  n144 = CollectionParameter() {collection=&MPC_GrillMoon, parameterName=Demon_FootPos_L}
  n145 = Reroute()
  n146 = ScalarParameter() {defaultValue=500, parameterName=FootGlowRadius}
  SphereMask(B=Add(A=ComponentMask(None=@n144) {b=true, g=true, r=true}, B=@n145), Radius=@n146) {hardnessPercent=0.8}
  n147 = CollectionParameter() {collection=&MPC_GrillMoon, parameterName=Demon_FootPos_R}
  SphereMask(B=Add(A=ComponentMask(None=@n147) {b=true, g=true, r=true}, B=@n145), Radius=@n146) {hardnessPercent=0.8}
 ]
}
